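# Niagara system for a confetti burst effect in the Sparks stage show effects set.
# It holds two emitters, LWE_Confetti_Sprite and LWE_Confetti_Mesh, named for sprite and mesh confetti variants.
# SystemState is the only module listed, with no renderers, data interfaces, dynamic inputs or user parameters shown.
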NiagaraSystem /SparksMaps/Effects/StageShow/Confetti/Particles/NS_LWE_Sparks_Confetti_Burst
dependencies {usedDataInterfaces [], usedDynamicInputs [], usedModules [&SystemState]}
emitters {LWE_Confetti_Mesh {rendererClasses [], renderers []}, LWE_Confetti_Sprite {rendererClasses [], renderers []}}
summary {
 systemName=NS_LWE_Sparks_Confetti_Burst
 emitters [{emitterName=LWE_Confetti_Sprite, rendererClasses []}, {emitterName=LWE_Confetti_Mesh, rendererClasses []}]
 userVariables []
}
